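# Animation Blueprint for the Plague glider, layering secondary motion on tassels, rips and hanging lanterns over an input pose.
# Trail controllers and wind-enabled AnimDynamics simulate a few tassel and rip bones, then Copy Bone cheaply shares that rotation with others.
# Lanterns combine sine-driven LanternShaking from the update event, tell-tale bone deltas and damped AnimDynamics on links and bodies.
# The TurnONEngines notify marks the glider fully deployed.
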AnimBlueprint /BRCosmetics/Gadgets/Assets/VinderTech_GliderChute/Glider_Plague/Meshes/Glider_Plague_AnimBP
target_skeleton=&/Glider_Plague_Skeleton
graphs [
 AnimGraph {nodes [
   Copy Bone - Source Bone: lanternTellTaleA_r - Target Bone: lantern_r!(ComponentPose:Component Space Pose Link Structure, Alpha:float=1) -> (Pose:Component Space Pose Link Structure) {
    is=title AnimGraphNode_CopyBone
    bCopyRotation=true
    alphaBoolBlend {blendOption=Linear}
    sourceBone {boneName=lanternTellTaleA_r}
    targetBone {boneName=lantern_r}
   }
   AnimGraphNode_SubInput_0 = Input Pose()
   n1 = Trail controller - Bone: tassleA1_l(ComponentPose=Local To Component(LocalPose=@AnimGraphNode_SubInput_0), Alpha=0.5) {
    bInvertChainBoneAxis=true
    bUsePlanarLimit=true
    chainBoneAxis=Z
    planarLimits [{drivingBone {boneName=tassleA_l}, planeTransform {location=<0 0 0>, rotation=<rot 0 0 170>, scale=<1 1 1>}}]
    rotationLimits [{limitMax=<180 180 180>}, {limitMax=<180 180 180>}]
    rotationOffsets [<0 0 0>, <0 0 0>]
    trailBone {boneName=tassleA1_l}
   }
   n2 = Anim Dynamics - Bone: tassleA_l(ComponentPose=@n1, Alpha=0.5, ExternalForce="0,0,0") {
    bEnableWind=true
    bGravityOverrideInSimSpace=true
    alphaBoolBlend {blendOption=Linear}
    boundBone {boneName=tassleA_l}
    componentAppliedLinearAccClamp=<300 300 300>
    componentLinearAccScale=<1 1 1>
    componentLinearVelScale=<1 1 1>
    physicsBodyDefinitions [{
      sphereCollisionRadius=0
      boundBone {boneName=tassleA_l}
      boxExtents=<5 5 2>
      constraintSetup {angularLimitsMax=<5 20 2>, angularLimitsMin=<-5 -20 -2>, angularTarget=<0 0 0>}
      localJointOffset=<-0.1 0 -4>
     }]
   }
   n3 = Copy Bone - Source Bone: tassleA_l - Target Bone: tassleC_l!(ComponentPose:Component Space Pose Link Structure=@n2, Alpha:float=1) -> (Pose:Component Space Pose Link Structure) {
    is=title AnimGraphNode_CopyBone
    bCopyRotation=true
    alphaBoolBlend {blendOption=Linear}
    sourceBone {boneName=tassleA_l}
    targetBone {boneName=tassleC_l}
   }
   n4 = Copy Bone - Source Bone: tassleA_l - Target Bone: tassleB_r!(ComponentPose:Component Space Pose Link Structure=@n3, Alpha:float=1) -> (Pose:Component Space Pose Link Structure) {
    is=title AnimGraphNode_CopyBone
    bCopyRotation=true
    alphaBoolBlend {blendOption=Linear}
    sourceBone {boneName=tassleA_l}
    targetBone {boneName=tassleB_r}
   }
   n5 = Copy Bone - Source Bone: tassleA_l - Target Bone: tassleD_r!(ComponentPose:Component Space Pose Link Structure=@n4, Alpha:float=1) -> (Pose:Component Space Pose Link Structure) {
    is=title AnimGraphNode_CopyBone
    bCopyRotation=true
    alphaBoolBlend {blendOption=Linear}
    sourceBone {boneName=tassleA_l}
    targetBone {boneName=tassleD_r}
   }
   n6 = Trail controller - Bone: tassleB1_l(ComponentPose=@n5, Alpha=0.5) {
    bInvertChainBoneAxis=true
    bUsePlanarLimit=true
    chainBoneAxis=Z
    planarLimits [{drivingBone {boneName=tassleB_l}, planeTransform {location=<0 0 0>, rotation=<rot 0 0 170>, scale=<1 1 1>}}]
    rotationLimits [{limitMax=<180 180 180>}, {limitMax=<180 180 180>}]
    rotationOffsets [<0 0 0>, <0 0 0>]
    trailBone {boneName=tassleB1_l}
   }
   n7 = Anim Dynamics - Bone: tassleB_l(ComponentPose=@n6, Alpha=0.5, ExternalForce="0,0,0") {
    bEnableWind=true
    bGravityOverrideInSimSpace=true
    alphaBoolBlend {blendOption=Linear}
    boundBone {boneName=tassleB_l}
    componentAppliedLinearAccClamp=<300 300 300>
    componentLinearAccScale=<1 1 1>
    componentLinearVelScale=<1 1 1>
    physicsBodyDefinitions [{
      sphereCollisionRadius=0
      boundBone {boneName=tassleB_l}
      boxExtents=<4 6 3>
      constraintSetup {angularLimitsMax=<5 20 2>, angularLimitsMin=<-5 -20 -2>, angularTarget=<0 0 0>}
      localJointOffset=<-0.1 0 -4>
     }]
   }
   n8 = Copy Bone - Source Bone: tassleB_l - Target Bone: tassleD_l!(ComponentPose:Component Space Pose Link Structure=@n7, Alpha:float=1) -> (Pose:Component Space Pose Link Structure) {
    is=title AnimGraphNode_CopyBone
    bCopyRotation=true
    alphaBoolBlend {blendOption=Linear}
    sourceBone {boneName=tassleB_l}
    targetBone {boneName=tassleD_l}
   }
   n9 = Copy Bone - Source Bone: tassleB_l - Target Bone: tassleA_r!(ComponentPose:Component Space Pose Link Structure=@n8, Alpha:float=1) -> (Pose:Component Space Pose Link Structure) {
    is=title AnimGraphNode_CopyBone
    bCopyRotation=true
    alphaBoolBlend {blendOption=Linear}
    sourceBone {boneName=tassleB_l}
    targetBone {boneName=tassleA_r}
   }
   n10 = Copy Bone - Source Bone: tassleB_l - Target Bone: tassleC_r!(ComponentPose:Component Space Pose Link Structure=@n9, Alpha:float=1) -> (Pose:Component Space Pose Link Structure) {
    is=title AnimGraphNode_CopyBone
    bCopyRotation=true
    alphaBoolBlend {blendOption=Linear}
    sourceBone {boneName=tassleB_l}
    targetBone {boneName=tassleC_r}
   }
   n11 = Copy Bone - Source Bone: tassleB_l - Target Bone: tassleA_f!(ComponentPose:Component Space Pose Link Structure=@n10, Alpha:float=1) -> (Pose:Component Space Pose Link Structure) {
    is=title AnimGraphNode_CopyBone
    bCopyRotation=true
    alphaBoolBlend {blendOption=Linear}
    sourceBone {boneName=tassleB_l}
    targetBone {boneName=tassleA_f}
   }
   n12 = Copy Bone - Source Bone: tassleB_l - Target Bone: tassleC_f!(ComponentPose:Component Space Pose Link Structure=@n11, Alpha:float=1) -> (Pose:Component Space Pose Link Structure) {
    is=title AnimGraphNode_CopyBone
    bCopyRotation=true
    alphaBoolBlend {blendOption=Linear}
    sourceBone {boneName=tassleB_l}
    targetBone {boneName=tassleC_f}
   }
   n13 = Copy Bone - Source Bone: tassleA_l - Target Bone: tassleB_f!(ComponentPose:Component Space Pose Link Structure=@n12, Alpha:float=1) -> (Pose:Component Space Pose Link Structure) {
    is=title AnimGraphNode_CopyBone
    bCopyRotation=true
    alphaBoolBlend {blendOption=Linear}
    sourceBone {boneName=tassleA_l}
    targetBone {boneName=tassleB_f}
   }
   n14 = Copy Bone - Source Bone: tassleA_l - Target Bone: tassleD_f!(ComponentPose:Component Space Pose Link Structure=@n13, Alpha:float=1) -> (Pose:Component Space Pose Link Structure) {
    is=title AnimGraphNode_CopyBone
    bCopyRotation=true
    alphaBoolBlend {blendOption=Linear}
    sourceBone {boneName=tassleA_l}
    targetBone {boneName=tassleD_f}
   }
   n15 = Trail controller - Bone: ripA1(ComponentPose=@n14, Alpha=0.1) {
    bUsePlanarLimit=true
    planarLimits [{drivingBone {boneName=ripA}, planeTransform {location=<0 0 0>, rotation=<rot 7.2 0 -180>, scale=<1 1 1>}}]
    rotationLimits [{limitMax=<180 180 180>}, {limitMax=<180 180 180>}]
    rotationOffsets [<0 0 0>, <0 0 0>]
    trailBone {boneName=ripA1}
   }
   n16 = Anim Dynamics - Bone: ripA(ComponentPose=@n15, Alpha=0.5, ExternalForce="0,0,0") {
    bEnableWind=true
    bGravityOverrideInSimSpace=true
    gravityScale=0.1
    alphaBoolBlend {blendOption=Linear}
    boundBone {boneName=ripA}
    componentAppliedLinearAccClamp=<300 300 300>
    componentLinearAccScale=<1 1 1>
    componentLinearVelScale=<1 1 1>
    physicsBodyDefinitions [{
      sphereCollisionRadius=0
      boundBone {boneName=ripA}
      boxExtents=<5 5 2>
      constraintSetup {angularLimitsMax=<2 45 12>, angularLimitsMin=<-2 -35 -12>, angularTarget=<0 0 0>}
      localJointOffset=<0 0 -4>
     }]
   }
   n17 = Trail controller - Bone: ripB1(
    ComponentPose=Copy Bone - Source Bone: ripA - Target Bone: ripE!(
     ComponentPose:Component Space Pose Link Structure=Copy Bone - Source Bone: ripA - Target Bone: ripC!(ComponentPose:Component Space Pose Link Structure=@n16, Alpha:float=1) -> (Pose:Component Space Pose Link Structure) {
      is=title AnimGraphNode_CopyBone
      bCopyRotation=true
      alphaBoolBlend {blendOption=Linear}
      sourceBone {boneName=ripA}
      targetBone {boneName=ripC}
     }
     Alpha:float=1
    ) -> (Pose:Component Space Pose Link Structure) {
     is=title AnimGraphNode_CopyBone
     bCopyRotation=true
     alphaBoolBlend {blendOption=Linear}
     sourceBone {boneName=ripA}
     targetBone {boneName=ripE}
    }
    Alpha=0.1
   ) {
    bUsePlanarLimit=true
    planarLimits [{drivingBone {boneName=ripA}, planeTransform {location=<0 0 0>, rotation=<rot 7.1999 0 -180>, scale=<1 1 1>}}]
    rotationLimits [{limitMax=<180 180 180>}, {limitMax=<180 180 180>}]
    rotationOffsets [<0 0 0>, <0 0 0>]
    trailBone {boneName=ripB1}
   }
   n18 = Anim Dynamics - Bone: ripB(ComponentPose=@n17, Alpha=0.5, ExternalForce="0,0,0") {
    bEnableWind=true
    bGravityOverrideInSimSpace=true
    gravityScale=0.1
    alphaBoolBlend {blendOption=Linear}
    boundBone {boneName=ripB}
    componentAppliedLinearAccClamp=<300 300 300>
    componentLinearAccScale=<1 1 1>
    componentLinearVelScale=<1 1 1>
    physicsBodyDefinitions [{
      sphereCollisionRadius=0
      boundBone {boneName=ripB}
      boxExtents=<5 5 4>
      constraintSetup {angularLimitsMax=<2 45 -5>, angularLimitsMin=<-2 -35 -10>, angularTarget=<0 0 0>}
      localJointOffset=<0 0 -4>
     }]
   }
   n19 = Anim Dynamics - Bone: ripD(
    ComponentPose=Copy Bone - Source Bone: ripB1 - Target Bone: ripD!(ComponentPose:Component Space Pose Link Structure=@n18, Alpha:float=0.5) -> (Pose:Component Space Pose Link Structure) {
     is=title AnimGraphNode_CopyBone
     bCopyRotation=true
     alphaBoolBlend {blendOption=Linear}
     sourceBone {boneName=ripB1}
     targetBone {boneName=ripD}
    }
    Alpha=0.5
    ExternalForce="0,0,0"
   ) {
    bEnableWind=true
    bGravityOverrideInSimSpace=true
    gravityScale=0.1
    alphaBoolBlend {blendOption=Linear}
    boundBone {boneName=ripD}
    componentAppliedLinearAccClamp=<300 300 300>
    componentLinearAccScale=<1 1 1>
    componentLinearVelScale=<1 1 1>
    physicsBodyDefinitions [{
      sphereCollisionRadius=0
      boundBone {boneName=ripD}
      boxExtents=<5 5 4>
      constraintSetup {angularLimitsMax=<2 45 -19.4>, angularLimitsMin=<-2 -35 -28>, angularTarget=<0 0 0>}
      localJointOffset=<0 0 -4>
     }]
   }
   n20 = Get LanternShaking() -> (LanternShaking:Rotator)
   n21 = "Transform (Modify) Bone - Bone: lantern_l"(ComponentPose=@n19, Alpha=1, Translation="0,0,0", Rotation=@n20, Scale="1,1,1") {
    rotationMode=BMM_Additive
    rotationSpace=BCS_BoneSpace
    translationMode=BMM_Additive
    translationSpace=BCS_BoneSpace
    alphaBoolBlend {blendOption=Linear}
    boneToModify {boneName=lantern_l}
   }
   n22 = "Transform (Modify) Bone - Bone: lantern_r"(ComponentPose=@n21, Alpha=1, Translation="0,0,0", Rotation=@n20, Scale="1,1,1") {
    rotationMode=BMM_Additive
    rotationSpace=BCS_BoneSpace
    translationMode=BMM_Additive
    translationSpace=BCS_BoneSpace
    alphaBoolBlend {blendOption=Linear}
    boneToModify {boneName=lantern_r}
   }
   n23 = Copy Bone Delta - Source Bone: lanternTellTaleA_l - Target Bone: lantern_l!(ComponentPose:Component Space Pose Link Structure=@n22, Alpha:float=1) -> (Pose:Component Space Pose Link Structure) {
    is=title AnimGraphNode_CopyBoneDelta
    bCopyRotation=true
    alphaBoolBlend {blendOption=Linear}
    sourceBone {boneName=lanternTellTaleA_l}
    targetBone {boneName=lantern_l}
   }
   n24 = Copy Bone Delta - Source Bone: lanternTellTaleA_r - Target Bone: lantern_r!(ComponentPose:Component Space Pose Link Structure=@n23, Alpha:float=1) -> (Pose:Component Space Pose Link Structure) {
    is=title AnimGraphNode_CopyBoneDelta
    bCopyRotation=true
    alphaBoolBlend {blendOption=Linear}
    sourceBone {boneName=lanternTellTaleA_r}
    targetBone {boneName=lantern_r}
   }
   n25 = Anim Dynamics - Bone: link_l(ComponentPose=@n24, Alpha=0.2, ExternalForce="0,0,0") {
    angularDampingOverride=0.99
    bGravityOverrideInSimSpace=true
    bOverrideAngularDamping=true
    alphaBoolBlend {blendOption=Linear}
    boundBone {boneName=link_l}
    componentAppliedLinearAccClamp=<400 400 400>
    componentLinearAccScale=<1 1 1>
    componentLinearVelScale=<1 1 1>
    physicsBodyDefinitions [{
      sphereCollisionRadius=0
      boundBone {boneName=link_l}
      boxExtents=<5 5 2>
      constraintSetup {twistAxis=AxisY, angularLimitsMax=<22 15 0>, angularLimitsMin=<-22 -15 0>, angularTarget=<0 0 0>}
      localJointOffset=<0 -0.1 -4>
     }]
   }
   n26 = Anim Dynamics - Bone: lantern_l(ComponentPose=@n25, Alpha=1, ExternalForce="0,0,0") {
    angularDampingOverride=0.998
    angularSpringConstant=500
    bGravityOverrideInSimSpace=true
    bOverrideAngularDamping=true
    alphaBoolBlend {blendOption=Linear}
    boundBone {boneName=lantern_l}
    componentAppliedLinearAccClamp=<400 400 400>
    componentLinearAccScale=<1 1 1>
    componentLinearVelScale=<1 1 1>
    physicsBodyDefinitions [{
      sphereCollisionRadius=0
      boundBone {boneName=lantern_l}
      boxExtents=<60 60 20>
      constraintSetup {twistAxis=AxisY, angularLimitsMax=<22 0 0>, angularLimitsMin=<-22 0 0>, angularTarget=<0 0 0>}
      localJointOffset=<0 -0.1 -28>
     }]
   }
   n27 = Anim Dynamics - Bone: link_r(ComponentPose=@n26, Alpha=0.2, ExternalForce="0,0,0") {
    angularDampingOverride=0.99
    bGravityOverrideInSimSpace=true
    bOverrideAngularDamping=true
    alphaBoolBlend {blendOption=Linear}
    boundBone {boneName=link_r}
    componentAppliedLinearAccClamp=<400 400 400>
    componentLinearAccScale=<1 1 1>
    componentLinearVelScale=<1 1 1>
    physicsBodyDefinitions [{
      sphereCollisionRadius=0
      boundBone {boneName=link_r}
      boxExtents=<5 5 2>
      constraintSetup {twistAxis=AxisY, angularLimitsMax=<22 15 0>, angularLimitsMin=<-22 -15 0>, angularTarget=<0 0 0>}
      localJointOffset=<0 -0.1 -4>
     }]
   }
   n28 = Anim Dynamics - Bone: lantern_r(ComponentPose=@n27, Alpha=1, ExternalForce="0,0,0") {
    angularDampingOverride=0.998
    bGravityOverrideInSimSpace=true
    bOverrideAngularDamping=true
    alphaBoolBlend {blendOption=Linear}
    boundBone {boneName=lantern_r}
    componentAppliedLinearAccClamp=<400 400 400>
    componentLinearAccScale=<1 1 1>
    componentLinearVelScale=<1 1 1>
    physicsBodyDefinitions [{
      sphereCollisionRadius=0
      boundBone {boneName=lantern_r}
      boxExtents=<66 56 22>
      constraintSetup {twistAxis=AxisY, angularLimitsMax=<22 0 0>, angularLimitsMin=<-22 0 0>, angularTarget=<0 0 0>}
      localJointOffset=<0 -0.1 -29>
     }]
   }
   Output Pose(Result=Component To Local(ComponentPose=@n28)) {layerGroup=None}
   Trail controller - Bone: lanternTellTaleB_r(Alpha=1) {
    bUsePlanarLimit=true
    chainBoneAxis=Z
    planarLimits [{drivingBone {boneName=lanternTellTaleA_r}, planeTransform {location=<0 0 0>, rotation=<rot 0 0 0>, scale=<1 1 1>}}]
    relaxationSpeedScaleInputProcessor {scale=0.2}
    rotationLimits [{limitMax=<180 180 180>}, {limitMax=<180 180 180>}]
    rotationOffsets [<0 0 0>, <0 0 0>]
    trailBone {boneName=lanternTellTaleB_r}
   }
   Trail controller - Bone: lanternTellTaleB_l(Alpha=1) {
    bActorSpaceFakeVel=true
    bUsePlanarLimit=true
    chainBoneAxis=Z
    planarLimits [{drivingBone {boneName=lanternTellTaleA_l}, planeTransform {location=<0 0 0>, rotation=<rot 0 0 0>, scale=<1 1 1>}}]
    relaxationSpeedScaleInputProcessor {scale=0.2}
    rotationLimits [{limitMax=<180 180 180>}, {limitMax=<180 180 180>}]
    rotationOffsets [<0 0 0>, <0 0 0>]
    trailBone {boneName=lanternTellTaleB_l}
   }
  ]}
 EventGraph {nodes [
   Event BlueprintUpdateAnimation()
   n1 = GetGameTimeInSeconds()
   n2 = "Sin (Radians)"(A=float * float(A=float * float(A=1, B=8), B=@n1))
   n3 = "Sin (Radians)"(A=float * float(A=float * float(A=1, B=13), B=@n1))
   n4 = "Sin (Radians)"(A=float * float(A=float * float(A=1, B=10), B=@n1))
   > Set LanternShaking(LanternShaking:Rotator=MakeRotator(Roll=float * float(A=@n4, B=13), Pitch=float * float(A=@n2, B=16), Yaw=float * float(A=@n3, B=12))) -> (Output_Get:Rotator)
   AnimNotify_TurnONEngines() {is=title K2Node_Event}
   > SetGliderFullyDeployed!(execute, self:Fort Glider Anim Instance Object Reference, bIsFullyDeployed:bool=true) -> (then)
  ]}
]
variables [
 DynamicsAlpha {type=boolean, default=true}
 LanternShaking {type=Rotator}
 MeshToCopy {type=SkeletalMeshComponent}
]
